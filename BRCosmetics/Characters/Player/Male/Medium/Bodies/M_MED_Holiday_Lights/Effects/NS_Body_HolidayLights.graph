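# Niagara effect for the Holiday Lights body cosmetic, adding glowing lights to the character's skeletal mesh.
# BigLights bursts seven persistent, light-only particles pinned by index to mesh sockets or bones.
# GeneralGlow continuously spawns short-lived sprites and lights on random bones, fading each in and out.
# Time-of-day curves scale brightness in both emitters, and user hand, foot and head flags feed per-particle visibility tags to hide covered parts.
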NiagaraSystem /BRCosmetics/Characters/Player/Male/Medium/Bodies/M_MED_Holiday_Lights/Effects/NS_Body_HolidayLights
dependencies {
 usedDataInterfaces [&NiagaraDataInterfaceCurve, &NiagaraDataInterfaceSkeletalMesh]
 usedDynamicInputs []
 usedModules [
  &self:SystemSpawnScript.SystemScriptSource.SystemScriptGraph.NiagaraNodeAssignment_0.SetVariables_#1
  &SystemState
 ]
}
emitters {
 BigLights {
  emitterUpdateScript [
   &EmitterState(Life Cycle Mode=::NewEnumerator1, Loop Duration=1, UseLoopDelay=false)
   &SpawnBurst_Instantaneous(
    Spawn Count=7
    Spawn Probability=&MakeCustomFloatFromBool(
     Bool=&SetBoolByIntComparison(A=$System.LoopCount, B=0, Comparison Type=::NewEnumerator0)
     True Float=0
     False Float=1
    )
    Use Spawn Probability=true
   )
  ]
  particleSpawnScript [/Niagara/Modules/Spawn/Initialization/InitializeParticle(
    Color=<rgba 1 0.9202 0.443 0.2>
    Module.Write Color=true
    Module.Write Lifetime=true
    Module.Write Mass=false
    Module.Write Material Random=true
    Module.Write Position=false
    Module.Write Scale=false
    Module.Write SpriteRotation=false
    Module.Write SpriteSize=true
    Position=&SimulationPosition()
    Sprite Size=&Vector2DFromFloat(Value=25)
   )]
  particleUpdateScript [
   &SkeletalMeshLocation(
    Bone / Socket Index=&ReturnExecIndex()
    Inherit Velocity=false
    MaskSpawnGroup=false
    Mesh Sampling Type=::NewEnumerator5
    Skeletal Mesh=$System.Skeletal Mesh
    Socket Sampling Mode=::NewEnumerator5
   )
   &ParticleState(Kill Particles When Lifetime Has Elapsed=false, Let Infinitely Lived Particles Die When Emitter Deactivates=true)
   &SolveForcesAndVelocity(Clamp Velocity=false, Limit Acceleration=false)
   &FN_Light_v3(
    Color=<rgba 1.9722 2 1.166 1>
    Create Color Attribute=true
    Create Exponent Attribute=true
    Create Radius Attribute=true
    Exponent=1.5
    Radius=20
    Scale Color:NiagaraFloat=&FloatFromCurve(
     FloatCurve=DataInterface() {propertyValues {Curve {keys [
         {interpMode=RCIM_Cubic, tangentMode=RCTM_User, time=0, value=1}
         {arriveTangent=-4.2769, interpMode=RCIM_Cubic, leaveTangent=-4.2769, tangentMode=RCTM_User, time=0.125, value=0.25}
         {arriveTangent=2, interpMode=RCIM_Cubic, leaveTangent=2, time=0.16, value=0.16}
         {interpMode=RCIM_Cubic, tangentMode=RCTM_User, time=0.5, value=1}
         {arriveTangent=-1.7863, interpMode=RCIM_Cubic, leaveTangent=-1.7863, time=0.8866, value=0.157}
         {arriveTangent=4.2765, interpMode=RCIM_Cubic, leaveTangent=4.2765, tangentMode=RCTM_User, time=0.9216, value=0.247}
         {interpMode=RCIM_Cubic, tangentMode=RCTM_User, time=1, value=1}
        ]}}}
     CurveIndex=Linked() {linkedVariable {name=NPC.FortniteNPC.TimeOfDay_0-1, type=NiagaraFloat}}
    )
   )
   /SetVariables(Particles.VisibilityTag=&self:PartHidden_Check_Stars(
     is Hand Hidden:NiagaraBool=$User.bisHandHidden
     is Foot Hidden:NiagaraBool=$User.bisFootHidden
     Selection:NiagaraInt32=&ReturnExecIndex()
    )) {isSetParameters=true, module=SetVariables_#2}
  ]
  rendererClasses [&NiagaraLightRendererProperties]
  renderers [{rendererClass=&NiagaraLightRendererProperties, rendererIndex=0}]
 }
 GeneralGlow {
  emitterUpdateScript [
   &EmitterState(Life Cycle Mode=::NewEnumerator1, Loop Duration=1, UseLoopDelay=false)
   &SpawnRate(SpawnRate=20, Use Spawn Probability=false)
  ]
  particleSpawnScript [
   &InitializeParticle(
    Color=<rgba 1 0.9321 0.526 0.05>
    Lifetime=2
    Module.Write Color=true
    Module.Write Lifetime=true
    Module.Write Mass=false
    Module.Write Material Random=true
    Module.Write Position=false
    Module.Write Scale=false
    Module.Write SpriteRotation=false
    Module.Write SpriteSize=true
    Position=&SimulationPosition()
    Sprite Size=&Vector2DFromFloat(Value=&RandomRangeFloat(Minimum=50, Maximum=100))
   )
   /SetVariables(
    Particles.PickBone:NiagaraInt32=&RandomRangeInt(Minimum=0, Maximum=22)
    Particles.RandOnOff:NiagaraFloat=&UniformAOrBFloat(A=0, Evaluation Type=::NewEnumerator0, Recalculate Random Each Loop=true, Override Seed=false)
   ) {isSetParameters=true}
  ]
  particleUpdateScript [
   &SkeletalMeshLocation(
    Bone / Socket Index=$Particles.PickBone
    Bone Sampling Mode=::NewEnumerator4
    Inherit Velocity=false
    MaskSpawnGroup=false
    Skeletal Mesh=$System.Skeletal Mesh
   )
   &ParticleState()
   &SolveForcesAndVelocity(Clamp Velocity=false, Limit Acceleration=false)
   &ScaleColor(
    Scale Alpha=&FloatFromCurve(
     FloatCurve=DataInterface() {propertyValues {Curve {keys [
         {interpMode=RCIM_Cubic, tangentMode=RCTM_User, time=0, value=4}
         {arriveTangent=0.4553, interpMode=RCIM_Cubic, leaveTangent=0.4553, tangentMode=RCTM_User, time=0.04, value=4}
         {arriveTangent=-40.101, interpMode=RCIM_Cubic, leaveTangent=-40.101, tangentMode=RCTM_User, time=0.1005, value=2.9056}
         {arriveTangent=-3.6881, interpMode=RCIM_Cubic, leaveTangent=-3.6881, tangentMode=RCTM_User, time=0.1294, value=0.1148}
         {arriveTangent=1.6875, interpMode=RCIM_Cubic, leaveTangent=1.6874, tangentMode=RCTM_User, time=0.208, value=0.25}
         {arriveTangent=0.4767, interpMode=RCIM_Cubic, leaveTangent=0.4767, tangentMode=RCTM_User, time=0.75, value=0.8}
         {arriveTangent=4.5813, interpMode=RCIM_Cubic, leaveTangent=4.5813, tangentMode=RCTM_User, time=0.9, value=0.175}
         {arriveTangent=2.916, interpMode=RCIM_Cubic, leaveTangent=2.916, tangentMode=RCTM_User, time=1, value=4}
        ]}}}
     CurveIndex=Linked() {linkedVariable {name=NPC.FortniteNPC.TimeOfDay_0-1, type=NiagaraFloat}}
     Scale Curve=&FloatFromCurve(FloatCurve=DataInterface() {propertyValues {Curve {keys [{interpMode=RCIM_Cubic, tangentMode=RCTM_User, time=0, value=0}, {interpMode=RCIM_Cubic, tangentMode=RCTM_User, time=0.5, value=1}, <at 1 0>]}}})
    )
    ScaleA=true
    ScaleRGB=false
   )
   &FN_Light_Attributes(
    Color=$Particles.Color
    Create Color Attribute=true
    Create Exponent Attribute=true
    Create Radius Attribute=true
    Exponent=1.5
    Radius=&Multiply_Float(
     A=&RandomRangeFloat(Minimum=10, Maximum=25)
     B=&FloatFromCurve(
      FloatCurve=DataInterface() {propertyValues {Curve {keys [
          {interpMode=RCIM_Cubic, tangentMode=RCTM_User, time=0.25, value=0}
          {arriveTangent=0.0076, interpMode=RCIM_Cubic, leaveTangent=0.0076, tangentMode=RCTM_User, time=0.35, value=1}
          {arriveTangent=0.0156, interpMode=RCIM_Cubic, leaveTangent=0.0156, tangentMode=RCTM_User, time=0.65, value=1}
          {interpMode=RCIM_Cubic, tangentMode=RCTM_User, time=0.75, value=0}
         ]}}}
      Scale Curve=$Particles.Initial.RandOnOff
     )
    )
    Scale Color=&FloatFromCurve(
     FloatCurve=DataInterface() {propertyValues {Curve {keys [
         {interpMode=RCIM_Cubic, tangentMode=RCTM_User, time=0.25, value=0}
         {arriveTangent=0.0076, interpMode=RCIM_Cubic, leaveTangent=0.0076, tangentMode=RCTM_User, time=0.4, value=1}
         {arriveTangent=0.0156, interpMode=RCIM_Cubic, leaveTangent=0.0156, tangentMode=RCTM_User, time=0.6, value=1}
         {interpMode=RCIM_Cubic, tangentMode=RCTM_User, time=0.75, value=0}
        ]}}}
     Scale Curve=&FloatFromCurve(
      FloatCurve=DataInterface() {propertyValues {Curve {keys [
          {interpMode=RCIM_Cubic, tangentMode=RCTM_User, time=0, value=4}
          {arriveTangent=0.4553, interpMode=RCIM_Cubic, leaveTangent=0.4553, tangentMode=RCTM_User, time=0.04, value=4}
          {arriveTangent=-40.101, interpMode=RCIM_Cubic, leaveTangent=-40.101, tangentMode=RCTM_User, time=0.1005, value=2.9056}
          {arriveTangent=-3.6881, interpMode=RCIM_Cubic, leaveTangent=-3.6881, tangentMode=RCTM_User, time=0.1294, value=0.1148}
          {arriveTangent=1.6875, interpMode=RCIM_Cubic, leaveTangent=1.6874, tangentMode=RCTM_User, time=0.208, value=0.25}
          {arriveTangent=0.4767, interpMode=RCIM_Cubic, leaveTangent=0.4767, tangentMode=RCTM_User, time=0.75, value=0.8}
          {arriveTangent=4.5813, interpMode=RCIM_Cubic, leaveTangent=4.5813, tangentMode=RCTM_User, time=0.9, value=0.175}
          {arriveTangent=2.916, interpMode=RCIM_Cubic, leaveTangent=2.916, tangentMode=RCTM_User, time=1, value=4}
         ]}}}
      CurveIndex=Linked() {linkedVariable {name=NPC.FortniteNPC.TimeOfDay_0-1, type=NiagaraFloat}}
     )
    )
   )
   /SetVariables(Particles.VisibilityTag=&self:PartHidden_Check(
     is Hand Hidden:NiagaraBool=$User.bisHandHidden
     is Head Hidden:NiagaraBool=$User.bisHeadHidden
     is Foot Hidden:NiagaraBool=$User.bisFootHidden
     Selection:NiagaraInt32=$Particles.PickBone
    )) {isSetParameters=true, module=SetVariables_#2}
  ]
  rendererClasses [&NiagaraSpriteRendererProperties, &NiagaraLightRendererProperties]
  renderers [
   {rendererClass=&NiagaraSpriteRendererProperties}
   {rendererClass=&NiagaraLightRendererProperties, rendererIndex=1}
  ]
 }
}
summary {
 systemName=NS_Body_HolidayLights
 emitters [
  {emitterName=BigLights, rendererClasses [&NiagaraLightRendererProperties]}
  {emitterName=GeneralGlow, rendererClasses [&NiagaraSpriteRendererProperties, &NiagaraLightRendererProperties]}
 ]
 userVariables [
  User.bisFootHidden {type=NiagaraBool}
  User.bisHandHidden {type=NiagaraBool}
  User.bisHeadHidden {type=NiagaraBool}
 ]
}
