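# Master material for the wires of the Holiday Lights character body, a two-sided, skeletal-mesh surface with parameter-driven PBR values.
# An optional switch animates twinkling by blending two emissive colors through quantized, panned UVs.
# A spline-thicken function supplies normals and vertex offset, with width scaled by camera distance.
# An optional ShoulderFix offsets masked vertices near the camera when PlayerPosition is close to the actor, while DisableShadows collapses shadow-pass vertices to the object position.
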Material /BRCosmetics/Characters/Player/Male/Medium/Bodies/M_MED_Holiday_Lights/Materials/M_MED_HolidayLights_Wires_Master
graph {
 properties {
  bUseMaterialAttributes=true
  bUsedWithMeshDeformer=true
  bUsedWithMorphTargets=true
  bUsedWithSkeletalMesh=true
  pixelDepthOffsetMode=PDOM_Legacy
  twoSided=true
  usedShadingModels=Default Lit
 }
 nodes [
  n1 = ScalarParameter() {defaultValue=1, group=TwinklingLights, parameterName=LightWireQuant}
  n2 = Multiply(A=Multiply(A=TextureCoordinate(), B=ScalarParameter() {defaultValue=10, group=TwinklingLights, parameterName=LightCount}), B=@n1)
  n3 = Panner(Coordinate=Divide(A=Floor(None=@n2), B=@n1), Speed=ScalarParameter() {defaultValue=1, group=TwinklingLights, parameterName=Speed}) {speedY=1}
  n4 = SphereMask(
   A=Frac(None=ComponentMask(None=@n3) {g=true})
   B=Constant() {r=0.5}
   Radius=ScalarParameter() {defaultValue=1, group=TwinklingLights, parameterName=LightRadius}
   Hardness=ScalarParameter() {group=TwinklingLights, parameterName=LightHardness}
  ) {attenuationRadius=0.5, hardnessPercent=0}
  Add(A=@n4, B=ScalarParameter() {group=TwinklingLights, parameterName=MinLight})
  n5 = VectorParameter() {parameterName=BaseColor}
  n6 = ScalarParameter() {parameterName=Metallic}
  n7 = ScalarParameter() {parameterName=Specular}
  n8 = ScalarParameter() {defaultValue=1, parameterName=Roughness}
  n9 = VectorParameter() {parameterName=Emissive2}
  n10 = VectorParameter() {parameterName=Emissive}
  n11 = Saturate(None=@n4)
  n12 = NamedRerouteDeclaration(None=StaticSwitchParameter(True=LinearInterpolate(A=@n9.RGB, B=@n10.RGB, Alpha=@n11), False=@n10.RGB) {parameterName=TwinklingLights}) {name=EmissiveColor, nodeColor=<rgba 0.7109 1 0 1>}
  n13 = ScalarParameter() {defaultValue=1, group=FrontendHacks, parameterName=LumenFix_EmissiveMult}
  n14 = ScalarParameter() {defaultValue=200, parameterName=depthMin}
  n15 = Subtract(A=Distance(A=WorldPosition().XYZ, B=CameraPositionWS()), B=@n14)
  n16 = Subtract(A=ScalarParameter() {defaultValue=601, parameterName=depthMax}, B=@n14)
  n17 = LinearInterpolate(
   A=ScalarParameter() {defaultValue=0.3, parameterName=widthMin}
   B=ScalarParameter() {defaultValue=1, parameterName=widthMax}
   Alpha=Saturate(None=Divide(A=@n15, B=@n16))
  )
  n18 = MaterialFunctionCall!(widthTip=@n17, Flatten=ScalarParameter() {defaultValue=0.5, parameterName="Flatten (S)"}, widthBase=@n17) {materialFunction=/Game/Characters/Player/Male/Medium/Bodies/M_MED_Neon/Materials/MF_splineThickenM}
  n19 = CollectionParameter() {collection=&FortniteMaterialParameters, parameterName=PlayerPosition}
  n20 = ComponentMask(None=@n19) {b=true, g=true, r=true}
  n21 = ActorPositionWS()
  n22 = Saturate(None=Multiply(A=Distance(A=@n20, B=@n21), B=ScalarParameter() {defaultValue=0.01, group=ShoulderFix, parameterName=DistThresh}))
  n23 = ComponentMask(None=MaterialFunctionCall() {materialFunction=&CameraDirectionVector}.Result) {b=true}
  n24 = MaterialFunctionCall(
   Offset Amount=LinearInterpolate(
    A=ScalarParameter() {defaultValue=20, group=ShoulderFix, parameterName=z}
    B=ScalarParameter() {defaultValue=9, group=ShoulderFix, parameterName=zv}
    Alpha=Abs(None=@n23)
   )
   World Position=WorldPosition().XYZ
  ) {materialFunction=/Engine/Functions/Engine_MaterialFunctions02/WorldPositionOffset/CameraOffset}
  n25 = SphereMask(
   A=CameraPositionWS()
   B=WorldPosition().XYZ
   Radius=ScalarParameter() {defaultValue=140, group=ShoulderFix, parameterName=rad}
   Hardness=ScalarParameter() {defaultValue=1, group=ShoulderFix, parameterName=hard}
  )
  n26 = PreSkinnedPosition()
  n27 = VectorParameter() {group=ShoulderFix, parameterName=s, defaultValue=<rgba -25 0 128 0>}
  n28 = ScalarParameter() {defaultValue=12, group=ShoulderFix, parameterName=rad2}
  n29 = ScalarParameter() {defaultValue=1, group=ShoulderFix, parameterName=hard2}
  n30 = Multiply(
   A=OneMinus(None=Round(None=@n22))
   B=Multiply(A=Multiply(A=@n24.Camera Offset, B=@n25), B=SphereMask(A=@n26, B=@n27.RGB, Radius=@n28, Hardness=@n29))
  ) {constA=1}
  n31 = Constant()
  n32 = Add(A=@n18.worldPositionOffset, B=ShadowReplace(Default=StaticSwitchParameter(True=@n30, False=@n31) {parameterName=ShoulderFix}, Shadow=@n31))
  n33 = ObjectPositionWS()
  n34 = WorldPosition()
  n35 = ShadowReplace(
   Default=@n32
   Shadow=LinearInterpolate(A=@n32, B=Subtract(A=@n33, B=@n34.XYZ), Alpha=ScalarParameter() {defaultValue=1, parameterName=DisableShadows})
  )
  MaterialFunctionCall(In=MakeMaterialAttributes!(
    BaseColor=NamedRerouteDeclaration(None=@n5.RGB) {name=BaseColor, nodeColor=<rgba 0 1 0.3203 1>}
    Metallic=NamedRerouteDeclaration(None=@n6) {name=Metallic, nodeColor=<rgba 1 0.3047 0 1>}
    Specular=NamedRerouteDeclaration(None=@n7) {name=Specular, nodeColor=<rgba 1 0.1641 0 1>}
    Roughness=NamedRerouteDeclaration(None=@n8) {name=Roughness, nodeColor=<rgba 1 0 0.7266 1>}
    Anisotropy
    EmissiveColor=Multiply(A=@n12, B=@n13)
    Opacity
    OpacityMask
    Normal=NamedRerouteDeclaration(None=@n18.normal) {name=Normal, nodeColor=<rgba 0.5469 0 1 1>}
    Tangent
    WorldPositionOffset=NamedRerouteDeclaration(None=@n35) {name=WPO, nodeColor=<rgba 0 0.4141 1 1>}
    SubsurfaceColor
    ClearCoat
    ClearCoatRoughness
    AmbientOcclusion
    Refraction
    CustomizedUVs_0
    CustomizedUVs_1
    CustomizedUVs_2
    CustomizedUVs_3
    CustomizedUVs_4
    CustomizedUVs_5
    CustomizedUVs_6
    CustomizedUVs_7
    PixelDepthOffset
    ShadingModel
    Displacement
   )) {materialFunction=&MF_TechArtOverrides}
  Reroute(None=@n5.RGB)
 ]
}
