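# Physics asset for the medium male Holiday Lights body cosmetic, covering secondary motion of its jacket, candies and pouch.
# Four two-link jacket chains per side use box bodies, with A bones constrained to spine_03 and B bones to A.
# Three candy capsules hang from spine_05 and a pouch from the left thigh twist, each limited mostly to one axis.
# Spine, left thigh and right upper arm capsules supply body shapes for anchoring or collision.
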PhysicsAsset /BRCosmetics/Characters/Player/Male/Medium/Bodies/M_MED_Holiday_Lights/Meshes/M_MED_Holiday_Lights_Physics
bodies {
 dyn_Candies_01 [{shapeType=Capsule, length=6.2092, radius=2.3501, center=<0.1157 -0.0164 -0.1862>, rotation=<rot -1.8045 -90 7.7747>}]
 dyn_Candies_02 [{shapeType=Capsule, length=6.1751, radius=2.4907, center=<-0.1496 -0.2579 -0.0358>, rotation=<rot 1.1567 -90 4.6369>}]
 dyn_Candies_03 [{shapeType=Capsule, length=5.9772, radius=2.5559, center=<-0.4535 -0.3015 0.0566>, rotation=<rot 7.0481 -90 7.4324>}]
 dyn_jacketA_l_01 [{shapeType=Box, extentX=2.2757, extentY=3.2714, extentZ=3.4107}]
 dyn_jacketA_l_02 [{shapeType=Box, extentX=2.2757, extentY=3.2714, extentZ=3.4107}]
 dyn_jacketA_l_03 [{shapeType=Box, extentX=2.2757, extentY=3.2714, extentZ=3.4107}]
 dyn_jacketA_l_04 [{shapeType=Box, extentX=2.2757, extentY=3.2714, extentZ=3.4107}]
 dyn_jacketA_r_01 [{shapeType=Box, extentX=2.2757, extentY=3.2714, extentZ=2.7501, center=<0 0 -0.3906>}]
 dyn_jacketA_r_02 [{shapeType=Box, extentX=2.2757, extentY=3.2714, extentZ=3.4107}]
 dyn_jacketA_r_03 [{shapeType=Box, extentX=2.2757, extentY=3.2714, extentZ=3.4107}]
 dyn_jacketA_r_04 [{shapeType=Box, extentX=2.2757, extentY=3.2714, extentZ=3.4107}]
 dyn_jacketB_l_01 [{shapeType=Box, extentX=2.4296, extentY=8.1405, extentZ=6.9383, center=<0.0002 0 -0.9397>}]
 dyn_jacketB_l_02 [{shapeType=Box, extentX=2.4296, extentY=8.1405, extentZ=6.9383, center=<0.0002 0 -0.4805>}]
 dyn_jacketB_l_03 [{shapeType=Box, extentX=2.4296, extentY=8.1405, extentZ=6.9383, center=<0.0002 0 -0.3842>}]
 dyn_jacketB_l_04 [{shapeType=Box, extentX=2.4296, extentY=8.1405, extentZ=6.9383, center=<0.0002 0 -0.9397>}]
 dyn_jacketB_r_01 [{shapeType=Box, extentX=2.4296, extentY=8.1405, extentZ=8.3015, center=<-0.0001 0 1.2433>}]
 dyn_jacketB_r_02 [{shapeType=Box, extentX=2.4296, extentY=8.1405, extentZ=6.7617, center=<-0.0001 0 1.0466>}]
 dyn_jacketB_r_03 [{shapeType=Box, extentX=2.4296, extentY=8.1405, extentZ=6.7617, center=<-0.0001 0 -0.1416>}]
 dyn_jacketB_r_04 [{shapeType=Box, extentX=2.4296, extentY=8.1405, extentZ=6.7617, center=<-0.0001 0 1.0466>}]
 dyn_pouch [{shapeType=Capsule, length=7.1011, radius=3.018, center=<3 0 -2>, rotation=<rot -2.34 -86.9427 0.0759>}]
 spine_02 [{shapeType=Capsule, length=6.758, radius=8.8604, center=<3.686 1.0301 0.1482>, rotation=<rot -2.4928 -89.0252 0.5362>}]
 spine_03 [{shapeType=Capsule, length=7.0538, radius=8.7676, center=<3.686 1.0301 0.1482>, rotation=<rot -2.4928 -89.0252 0.5362>}]
 spine_05 [{shapeType=Capsule, length=12.638, radius=11.0922, center=<1.512 -0.8038 0.2468>, rotation=<rot 0 21.8029 179.8598>}]
 thigh_twist_01_l [{shapeType=Capsule, length=22.1307, radius=7.8693, rotation=<rot 85.5795 0 0>}]
 upperarm_twist_02_r [{shapeType=Capsule, length=16.7226, radius=3.9233, rotation=<rot -86.5211 -0.0002 3.5581>}]
}
constraints [
 {bone1Name=dyn_jacketA_l_01, bone2Name=spine_03, swing1LimitDegrees=1, swing2LimitDegrees=8, twistLimitDegrees=1}
 {bone1Name=dyn_jacketB_l_01, bone2Name=dyn_jacketA_l_01, swing1LimitDegrees=1, swing2LimitDegrees=8, twistLimitDegrees=1}
 {bone1Name=dyn_jacketA_l_02, bone2Name=spine_03, swing1LimitDegrees=1, swing2LimitDegrees=8, twistLimitDegrees=1}
 {bone1Name=dyn_jacketB_l_02, bone2Name=dyn_jacketA_l_02, swing1LimitDegrees=1, swing2LimitDegrees=8, twistLimitDegrees=1}
 {bone1Name=dyn_jacketA_l_03, bone2Name=spine_03, swing1LimitDegrees=1, swing2LimitDegrees=8, twistLimitDegrees=1}
 {bone1Name=dyn_jacketB_l_03, bone2Name=dyn_jacketA_l_03, swing1LimitDegrees=1, swing2LimitDegrees=8, twistLimitDegrees=1}
 {bone1Name=dyn_jacketA_l_04, bone2Name=spine_03, swing1LimitDegrees=1, swing2LimitDegrees=8, twistLimitDegrees=1}
 {bone1Name=dyn_jacketB_l_04, bone2Name=dyn_jacketA_l_04, swing1LimitDegrees=1, swing2LimitDegrees=8, twistLimitDegrees=1}
 {bone1Name=dyn_jacketA_r_01, bone2Name=spine_03, swing1LimitDegrees=1, swing2LimitDegrees=8, twistLimitDegrees=1}
 {bone1Name=dyn_jacketB_r_01, bone2Name=dyn_jacketA_r_01, swing1LimitDegrees=1, swing2LimitDegrees=8, twistLimitDegrees=1}
 {bone1Name=dyn_jacketA_r_02, bone2Name=spine_03, swing1LimitDegrees=1, swing2LimitDegrees=8, twistLimitDegrees=1}
 {bone1Name=dyn_jacketB_r_02, bone2Name=dyn_jacketA_r_02, swing1LimitDegrees=1, swing2LimitDegrees=8, twistLimitDegrees=1}
 {bone1Name=dyn_jacketA_r_03, bone2Name=spine_03, swing1LimitDegrees=1, swing2LimitDegrees=8, twistLimitDegrees=1}
 {bone1Name=dyn_jacketB_r_03, bone2Name=dyn_jacketA_r_03, swing1LimitDegrees=1, swing2LimitDegrees=8, twistLimitDegrees=1}
 {bone1Name=dyn_jacketA_r_04, bone2Name=spine_03, swing1LimitDegrees=1, swing2LimitDegrees=8, twistLimitDegrees=1}
 {bone1Name=dyn_jacketB_r_04, bone2Name=dyn_jacketA_r_04, swing1LimitDegrees=1, swing2LimitDegrees=8, twistLimitDegrees=1}
 {bone1Name=dyn_pouch, bone2Name=thigh_twist_01_l, swing1LimitDegrees=20, swing2LimitDegrees=1, twistLimitDegrees=1}
 {bone1Name=dyn_Candies_01, bone2Name=spine_05, swing1LimitDegrees=1, swing2LimitDegrees=1, twistLimitDegrees=25}
 {bone1Name=dyn_Candies_02, bone2Name=spine_05, swing1LimitDegrees=1, swing2LimitDegrees=1, twistLimitDegrees=30}
 {bone1Name=dyn_Candies_03, bone2Name=spine_05, swing1LimitDegrees=1, swing2LimitDegrees=1, twistLimitDegrees=40}
]
