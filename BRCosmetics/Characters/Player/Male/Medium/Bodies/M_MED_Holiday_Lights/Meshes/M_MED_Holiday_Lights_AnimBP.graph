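# Post-process-style cosmetic Anim Blueprint for the male medium Holiday Lights body, adding secondary motion to a pose copied from the attached parent mesh.
# Additive offsets nudge both jacket fix bones, then spring-driven AnimDynamics swings the pouch relative to the pelvis.
# A RigidBody node simulates the outfit's physics asset in spine_03 space with damped component motion, before a final additive pouch adjustment.
# The EventGraph is empty and the declared variables are unused.
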AnimBlueprint /BRCosmetics/Characters/Player/Male/Medium/Bodies/M_MED_Holiday_Lights/Meshes/M_MED_Holiday_Lights_AnimBP
target_skeleton=&/M_MED_Holiday_Lights_Skeleton
graphs [
 AnimGraph {nodes [
   n1 = "Transform (Modify) Bone - Bone: dyn_jacket_fix_r"(
    ComponentPose=Local To Component(LocalPose=Copy Pose From Mesh() {bUseAttachedParent=true})
    Alpha=1
    Translation="1,0,0"
    Rotation="0,0,0"
    Scale="1,1,1"
   ) {
    rotationMode=BMM_Additive
    rotationSpace=BCS_BoneSpace
    scaleMode=BMM_Additive
    scaleSpace=BCS_BoneSpace
    translationMode=BMM_Additive
    translationSpace=BCS_BoneSpace
    alphaBoolBlend {blendOption=Linear}
    boneToModify {boneName=dyn_jacket_fix_r}
    translation {x=1}
   }
   n2 = "Transform (Modify) Bone - Bone: dyn_jacket_fix_l"(ComponentPose=@n1, Alpha=1, Translation="-1,0,0", Rotation="0,0,0", Scale="1,1,1") {
    rotationMode=BMM_Additive
    rotationSpace=BCS_BoneSpace
    scaleMode=BMM_Additive
    scaleSpace=BCS_BoneSpace
    translationMode=BMM_Additive
    translationSpace=BCS_BoneSpace
    alphaBoolBlend {blendOption=Linear}
    boneToModify {boneName=dyn_jacket_fix_l}
    translation {x=-1}
   }
   n3 = Anim Dynamics - Bone: dyn_pouch(ComponentPose=@n2, Alpha=1, ExternalForce="0,0,0") {
    angularDampingOverride=0.999
    angularSpringConstant=1000
    bAngularSpring=true
    bGravityOverrideInSimSpace=true
    bOverrideAngularDamping=true
    bUseGravityOverride=true
    simulationSpace=BoneRelative
    alphaBoolBlend {blendOption=Linear}
    boundBone {boneName=dyn_pouch}
    componentAppliedLinearAccClamp=<800 800 800>
    componentLinearAccScale=<1 1 1>
    componentLinearVelScale=<1 1 1>
    gravityOverride {z=-980}
    physicsBodyDefinitions [{
      sphereCollisionRadius=0
      boundBone {boneName=dyn_pouch}
      boxExtents=<14 21 14>
      constraintSetup {angularLimitsMax=<8 6 8>, angularLimitsMin=<-8 -0.5 -8>, angularTarget=<0 0 0>}
      localJointOffset=<5 0 0>
     }]
    relativeSpaceBone {boneName=pelvis}
   }
   n4 = RigidBody(ComponentPose=@n3) {
    bClampLinearTranslationLimitToRefPose=true
    bOverrideWorldGravity=true
    overridePhysicsAsset=&/M_MED_Holiday_Lights_Physics
    simulationSpace=BaseBoneSpace
    alphaBoolBlend {blendOption=Linear}
    baseBoneRef {boneName=spine_03}
    componentAppliedLinearAccClamp=<1200 1200 1200>
    componentLinearAccScale=<0.35 0.45 0.45>
    componentLinearVelScale=<0.3 0.3 0.3>
    overrideWorldGravity {z=-980}
   }
   n5 = "Transform (Modify) Bone - Bone: dyn_pouch"!(ComponentPose=@n4, Alpha=1, Rotation="2.8,0,0") -> (Pose) {
    rotationMode=BMM_Additive
    rotationSpace=BCS_BoneSpace
    scaleSpace=BCS_BoneSpace
    translationMode=BMM_Additive
    translationSpace=BCS_BoneSpace
    alphaBoolBlend {blendOption=Linear}
    boneToModify {boneName=dyn_pouch}
    rotation {pitch=2.8}
    translation {x=0.2}
   }
   Output Pose(Result=Component To Local(ComponentPose=@n5)) {layerGroup=None}
  ]}
 EventGraph {links [], nodes []}
]
variables [
 Shoulder_rot_down_l {type=number}
 Shoulder_rot_down_r {type=number}
 Skel_Mesh {type=SkeletalMeshComponent}
]
