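# Physics asset for a part of the M_MED_Holiday_Lights male player cosmetic, covering the head and a three-bone dynamic hat chain.
# A capsule on the head anchors the chain, with boxes on the first two hat bones and a small capsule on the tip.
# Each hat bone is constrained to the one before it with the same tight swing and twist limits, allowing only slight bending along the hat.
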PhysicsAsset /BRCosmetics/Characters/Player/Male/Medium/Bodies/M_MED_Holiday_Lights/Meshes/Parts/M_MED_Holiday_Lights_Physics
bodies {
 dyn_hat_00 [{shapeType=Box, extentX=7.5, extentY=2.2158, extentZ=5.7692, center=<0 -0.184 0>}]
 dyn_hat_01 [{shapeType=Box, extentX=7.5, extentY=2.1404, extentZ=6.279, center=<0 0.3185 -0.7519>, rotation=<rot 0 0.0001 3.9608>}]
 dyn_hat_02 [{shapeType=Capsule, length=3.5043, radius=2.2076, center=<0 0 -1.7353>}]
 head [{shapeType=Capsule, length=9.0977, radius=10.1416, center=<8.0228 3.5966 -0.5489>, rotation=<rot 24.5343 88.8803 -91.5034>}]
}
constraints [
 {bone1Name=dyn_hat_00, bone2Name=head, swing1LimitDegrees=10, swing2LimitDegrees=10, twistLimitDegrees=20}
 {bone1Name=dyn_hat_01, bone2Name=dyn_hat_00, swing1LimitDegrees=10, swing2LimitDegrees=10, twistLimitDegrees=20}
 {bone1Name=dyn_hat_02, bone2Name=dyn_hat_01, swing1LimitDegrees=10, swing2LimitDegrees=10, twistLimitDegrees=20}
]
